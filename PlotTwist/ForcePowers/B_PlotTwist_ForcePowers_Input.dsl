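; Blueprint for Force Powers input in the PlotTwist project.
; It handles a Force button pressed event by printing a debug message that the button was pressed.
; BeginPlay and Tick events are present but empty, so no power logic runs in this graph.
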(event EventBeginPlay)

(event EventTick
  (params (DeltaSeconds :type Float%20%28single-precision%29)))

(event PlotTwist|EventOnForceButtonPressed
  (Development|PrintString "The button has been pressed!"))
