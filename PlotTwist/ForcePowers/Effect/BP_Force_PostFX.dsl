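; BP_Force_PostFX is a short-lived actor that plays a timed post-process effect for a force power, visible only to the owning local player.
; Construction enables its PostProcess component only when the PlayerPawn variable is valid and locally controlled, disabling it otherwise.
; On BeginPlay it waits DelayTime, then a timeline drives blend weight from a curve scaled by PostFX_Intensity and destroys the actor on finish.
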(event EventBeginPlay
  (Utilities|FlowControl|Delay (Variables|Default|GetDelayTime))
  (bind (_direction _postweightcurve) (Timeline|Timeline PlayFromStart)
    (:Update
      (Variables|/Script/Engine.PostProcessComponent|SetBlendWeight (* _postweightcurve (Variables|Default|GetPostFX_Intensity)) (Variables|Default|GetPostProcess)))
    (:Finished
      (Actor|DestroyActor))))

(fn ConstructionScript
  (bind _postprocess (Variables|Default|GetPostProcess))
  (if (and (Utilities|IsValid (Variables|Default|GetPlayerPawn) :!class /Script/Engine.KismetSystemLibrary) (Pawn|IsLocallyControlled (Variables|Default|GetPlayerPawn)))
    (Variables|/Script/Engine.PostProcessComponent|SetbEnabled true _postprocess)
    (else
      (Variables|/Script/Engine.PostProcessComponent|SetbEnabled false _postprocess))))
