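# Additive, unlit, two-sided Force energy material for Niagara mesh particles, with depth-based fading.
# Panning chromatic and noise textures, with UVs offset through named reroutes alongside a noise-driven Distortion value, blend two blue tints multiplied by particle color.
# An alpha chain combines view-angle fade, DepthFade (bypassed on forward and mobile), a sine V gradient, square edge mask and particle alpha.
# View-angle fade and particle alpha also drive a refraction lerp.
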Material /PlotTwist/ForcePowers/Effect/M_Force_Energy_DepthFade
graph {
 properties {
  bUsedWithNiagaraMeshParticles=true
  blendMode=BLEND_Additive
  pixelDepthOffsetMode=PDOM_Legacy
  refractionMethod=RM_IndexOfRefraction
  shadingModel=MSM_Unlit
  twoSided=true
  usedShadingModels=Unlit
 }
 nodes [
  Constant() {r=1}
  n1 = ParticleColor()
  n2 = MaterialFunctionCall("FadeSensitivity (optional)"=Constant() {r=3}) {materialFunction=&FadeBasedOnViewAngle}
  LinearInterpolate(B=ScalarParameter() {defaultValue=1.05, parameterName=Refraction}, Alpha=Multiply(A=@n1.A, B=@n2.Result)) {constA=1, constB=1.1}
  n3 = Panner(Coordinate=Add(A=NamedRerouteUsage(), B=TextureCoordinate() {uTiling=1.5})) {speedY=4}
  n4 = Add(A=NamedRerouteUsage(), B=TextureCoordinate() {uTiling=2})
  n5 = Add(
   A=Power(Base=TextureSample(UVs=@n3) {texture=/Game/Effects/Fort_Effects/Textures/Tile/T_Smooth_Tile_Chromatic}.RGB, Exp=Constant() {r=5}) {constExponent=3}
   B=TextureSample(UVs=Panner(Coordinate=@n4) {speedY=3}) {texture=/Game/Effects/Fort_Effects/Textures/Tile/T_Noise_03}.RGB
  )
  n6 = LinearInterpolate(
   A=VectorParameter() {defaultValue=<rgba 0.844 1.3834 2 1>}.RGB
   B=VectorParameter() {parameterName=Param1, defaultValue=<rgba 0.314 0.4698 1 1>}.RGB
   Alpha=Normalize(VectorInput=@n5)
  )
  MaterialFunctionCall(Emissive=Multiply(A=Multiply(A=@n6, B=@n5), B=@n1.RGB)) {materialFunction=&MF_FX_EmissiveAdjustments}
  n7 = ScalarParameter() {defaultValue=5000, parameterName=FadeDistance}
  n8 = DepthFade() {fadeDistanceDefault=150}
  Multiply(A=OneMinus(None=DepthFade(FadeDistance=@n7) {fadeDistanceDefault=600}), B=@n8)
  n9 = Constant() {r=1}
  n10 = Multiply(A=ShadingPathSwitch(Default=@n8, Deferred=@n8, Forward=@n9, Mobile=@n9), B=@n2.Result)
  n11 = TextureCoordinate()
  n12 = Multiply(A=MaterialFunctionCall(Float2=@n11) {materialFunction=&BreakOutFloat2Components}.G) {constB=0.5}
  n13 = Power(Base=Saturate(None=Sine(None=@n12)), Exp=Constant() {r=3})
  n14 = Constant() {r=2}
  n15 = Constant() {r=1}
  Multiply(
   A=@n1.A
   B=Multiply(
    A=Multiply(A=@n10, B=@n13)
    B=MaterialFunctionCall!(UVs=@n11, Mult=@n14, Exp=@n15) {materialFunction=/Game/Effects/Fort_Effects/MaterialFunctions/MF_SquareEdgeMask}.SquareMask
   )
  )
  n16 = TextureCoordinate()
  n17 = AppendVector(A=Constant() {r=3}, B=Constant() {r=0.5})
  n18 = Constant() {r=-0.5}
  n19 = Constant() {r=0.7}
  n20 = TextureSample(UVs=Panner(Coordinate=Multiply(A=@n16, B=@n17), Time=NamedRerouteUsage(), Speed=AppendVector(A=@n18, B=@n19)) {speedY=1}) {texture=&T_NoisePack_01}
  NamedRerouteDeclaration(None=Divide(A=AppendVector(A=@n20.R, B=@n20.R), B=ScalarParameter() {defaultValue=7, parameterName=Param_9})) {name=Distortion, nodeColor=<rgba 1 0.5156 0 1>}
  TextureSample(UVs=@n3) {samplerType=SAMPLERTYPE_Normal, texture=/Game/Characters/Player/Female/Medium/Bodies/F_MED_Bistro_Astronaut/FX/T_GaussianNoise_Normal}
 ]
}
