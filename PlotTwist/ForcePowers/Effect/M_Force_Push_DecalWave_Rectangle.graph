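# Additive, unlit material for Niagara mesh particles that draws a rectangular Force Push wave as a fake decal.
# World position behind translucency, transformed to local or particle-relative space, supplies decal UVs and a noise-softened box edge mask.
# Panned noise layers, eroded by a dynamic parameter and shaped by sine gradients, drive opacity and two-color emissive, plus camera-faded refraction.
# UseAnimPanning blends panning from time to a particle-driven value, with per-particle random offset.
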Material /PlotTwist/ForcePowers/Effect/M_Force_Push_DecalWave_Rectangle
graph {
 properties {
  bDisableDepthTest=true
  bUsedWithNiagaraMeshParticles=true
  blendMode=BLEND_Additive
  pixelDepthOffsetMode=PDOM_Legacy
  refractionMethod=RM_IndexOfRefraction
  shadingModel=MSM_Unlit
  twoSided=true
  usedShadingModels=Unlit
 }
 nodes [
  Constant() {r=1}
  Constant() {r=1}
  Constant() {r=1}
  MaterialFunctionCall(Emissive=NamedRerouteUsage()) {materialFunction=&MF_FX_EmissiveAdjustments}
  n1 = DotProduct(A=Constant3Vector() {constant=<rgba 0 0 1 0>}, B=ObjectOrientation())
  Multiply(A=NamedRerouteUsage(), B=Power(Base=Abs(None=@n1), Exp=Constant() {r=60}))
  n2 = Constant() {r=1}
  NamedRerouteDeclaration(None=MaterialFunctionCall(X=@n2, Y=@n2, Z=@n2) {materialFunction=&MakeFloat3}.Result) {name=SizeScalar, nodeColor=<rgba 0 0.6744 1 1>}
  n3 = Divide(A=ComponentMask(None=NamedRerouteUsage()) {g=true, r=true}, B=Constant() {r=100})
  n4 = Constant() {r=0.8}
  n5 = NamedRerouteUsage()
  NamedRerouteDeclaration(None=MaterialFunctionCall(UVs=Add(A=@n3) {constB=0.5}, Texture Scale=Multiply(A=@n4, B=@n5)) {materialFunction=&ScaleUVsByCenter}.UVs) {name=PD_DecalUV, nodeColor=<rgba 1 0.1406 0 1>}
  n6 = NamedRerouteUsage()
  n7 = NamedRerouteUsage()
  n8 = ScalarParameter() {defaultValue=0.5, parameterName=Param3}
  n9 = Multiply(
   A=ScalarParameter() {defaultValue=1, parameterName=Param_1}
   B=DynamicParameter() {defaultValue=<rgba 1 0 1 1>, paramNames [Refraction,Erode,AnimPanning,Tile]}.Tile
  )
  n10 = Panner(
   Coordinate=Multiply(A=Add(A=@n6, B=@n7), B=AppendVector(A=@n8, B=@n9))
   Time=NamedRerouteUsage()
   Speed=AppendVector(A=ScalarParameter() {defaultValue=-1, parameterName=s}, B=ScalarParameter() {defaultValue=0.2, parameterName=Param_5})
  ) {speedY=1}
  n11 = Multiply(
   A=TextureSample(UVs=@n10) {samplerType=SAMPLERTYPE_Grayscale, texture=&LowResBlurredNoise}.R
   B=ScalarParameter() {defaultValue=3, parameterName=OpacityMulti}
  )
  n12 = NamedRerouteUsage()
  n13 = NamedRerouteUsage()
  n14 = ScalarParameter() {defaultValue=0.2}
  n15 = Panner(
   Coordinate=Multiply(A=Add(A=@n12, B=@n13), B=AppendVector(A=@n14, B=@n9))
   Time=NamedRerouteUsage()
   Speed=AppendVector(A=ScalarParameter() {defaultValue=-0.5, parameterName=s1}, B=ScalarParameter() {defaultValue=-0.2, parameterName=Param_6})
  ) {speedY=1}
  n16 = Multiply(
   A=Add(A=NamedRerouteUsage(), B=NamedRerouteUsage())
   B=AppendVector(A=ScalarParameter() {defaultValue=0.3, parameterName=Param35}, B=ScalarParameter() {defaultValue=1, parameterName=Param_41})
  )
  n17 = NamedRerouteUsage()
  n18 = AppendVector(A=ScalarParameter() {defaultValue=-0.3, parameterName=s3}, B=ScalarParameter() {defaultValue=0.2, parameterName=Param_55})
  n19 = Add(
   A=Add(
    A=TextureSample(UVs=@n15) {texture=/Game/Effects/Fort_Effects/Textures/Tile/T_Noise_03}.RGB
    B=TextureSample(UVs=@n10) {texture=/Game/Effects/Fort_Effects/Textures/Tile/T_Noise_02}.RGB
   )
   B=TextureSample(UVs=Panner(Coordinate=@n16, Time=@n17, Speed=@n18) {speedY=1}) {texture=&T_Noise_03}.RGB
  )
  n20 = DynamicParameter() {defaultValue=<rgba 1 0 1 1>, paramNames [Refraction,Erode,AnimPanning,Tile]}
  n21 = Multiply(A=@n11, B=Saturate(None=Subtract(A=@n19, B=@n20.Erode)))
  n22 = MaterialFunctionCall(UV Channel=NamedRerouteUsage()) {materialFunction=&LinearGradient}
  n23 = Sine(None=Multiply(A=@n22.VGradient) {constB=0.5})
  n24 = Constant() {r=2}
  n25 = Sine(None=Multiply(A=@n22.UGradient) {constB=0.5})
  n26 = Constant() {r=1.8}
  n27 = Saturate(None=Multiply(A=Power(Base=@n23, Exp=@n24), B=Power(Base=@n25, Exp=@n26)))
  n28 = Multiply(A=Multiply(A=Multiply(A=@n21, B=@n27), B=NamedRerouteUsage()), B=NamedRerouteUsage())
  n29 = Multiply(A=@n28, B=@n20.Refraction)
  n30 = ScalarParameter() {defaultValue=3500, parameterName=Param_11}
  n31 = LinearInterpolate(Alpha=Saturate(None=MaterialFunctionCall(Distance=@n30) {materialFunction=&Near_Camera_Fade}.Result)) {constA=1, constB=0}
  NamedRerouteDeclaration(None=LinearInterpolate(Alpha=Multiply(A=@n29, B=@n31)) {constA=1, constB=1.1}) {name=Refraction, nodeColor=<rgba 0.3511 0.5415 1 1>}
  n32 = Add(A=NamedRerouteUsage(), B=Multiply(A=Constant3Vector() {constant=<rgba 0 0 0 1>}, B=NamedRerouteUsage())) {constB=50}
  n33 = NamedRerouteUsage()
  n34 = ScalarParameter() {defaultValue=16, parameterName=EdgeFalloff}
  n35 = Divide(A=@n34) {constB=4}
  n36 = TextureSample(UVs=Panner(Coordinate=TextureCoordinate() {uTiling=2}) {speedX=0.2, speedY=0.2}) {samplerType=SAMPLERTYPE_Grayscale, texture=&LowResBlurredNoise}
  NamedRerouteDeclaration(None=MaterialFunctionCall!(
    A=Divide(A=@n32, B=@n33)
    B
    Bounds=VectorParameter() {parameterName=Bounds, defaultValue=<rgba 100 100 80 1>}.RGB
    Edge Falloff=LinearInterpolate(A=@n35, B=@n34, Alpha=@n36.R)
   ) {materialFunction=/Engine/Functions/Engine_MaterialFunctions01/Gradient/BoxMask-3D}.Result) {name=PD_EdgeOpacity, nodeColor=<rgba 1 0.1406 0 1>}
  n37 = MaterialFunctionCall() {materialFunction=&WorldPositionBehindTranslucency}
  n38 = Subtract(
   A=MaterialFunctionCall() {materialFunction=&WorldPositionBehindTranslucency}.Result
   B=TransformPosition(None=ParticlePositionWS()) {transformSourceType=TRANSFORMPOSSOURCE_World, transformType=TRANSFORMPOSSOURCE_Instance}
  )
  NamedRerouteDeclaration(None=LinearInterpolate(
    A=TransformPosition(None=@n37.Result) {transformSourceType=TRANSFORMPOSSOURCE_World}
    B=TransformPosition(None=@n38) {transformSourceType=TRANSFORMPOSSOURCE_World, transformType=TRANSFORMPOSSOURCE_Instance}
    Alpha=ScalarParameter() {parameterName=UseAnimPanning}
   )) {name=Position, nodeColor=<rgba 0 0.0859 1 1>}
  n39 = Time()
  n40 = DynamicParameter() {defaultValue=<rgba 1 0 1 1>, paramNames [Refraction,Erode,AnimPanning,Tile]}
  n41 = ScalarParameter() {parameterName=UseAnimPanning}
  NamedRerouteDeclaration(None=Add(A=LinearInterpolate(A=@n39, B=@n40.AnimPanning, Alpha=@n41), B=ParticleRandom())) {name=RandomTimeOffset, nodeColor=<rgba 0.125 1 0 1>}
  n42 = NamedRerouteUsage()
  n43 = AppendVector(A=ScalarParameter() {defaultValue=1, parameterName=Param352}, B=ScalarParameter() {defaultValue=0.5, parameterName=Param_413})
  n44 = ScalarParameter() {defaultValue=-1, parameterName=s31}
  n45 = ScalarParameter() {parameterName=Param_551}
  n46 = TextureSample(UVs=Panner(Coordinate=Multiply(A=@n42, B=@n43), Time=NamedRerouteUsage(), Speed=AppendVector(A=@n44, B=@n45)) {speedY=1}) {texture=&T_NoisePack_01}
  NamedRerouteDeclaration(None=Divide(A=AppendVector(A=@n46.R, B=@n46.R), B=ScalarParameter() {defaultValue=10, parameterName=Param_9})) {name=Distortion, nodeColor=<rgba 1 0.5156 0 1>}
  NamedRerouteDeclaration(None=Constant() {r=1}) {name=Fade, nodeColor=<rgba 1 0.5354 0.8627 1>}
  n47 = ParticleColor()
  n48 = LinearInterpolate(
   A=VectorParameter() {parameterName=SubColor, defaultValue=<rgba 0.4763 0.06 5 0>}.RGB
   B=VectorParameter() {parameterName=MainColor, defaultValue=<rgba 4.741 9.454 25 0>}.RGB
   Alpha=Saturate(None=@n11)
  )
  n49 = Desaturation(None=@n19, Fraction=ScalarParameter() {parameterName=Desatuation})
  NamedRerouteDeclaration(None=Multiply(A=@n47.RGB, B=Multiply(A=@n48, B=@n49))) {name=Emissive, nodeColor=<rgba 0.0833 0.1783 1 0>}
  NamedRerouteDeclaration(None=Multiply(A=@n47.A, B=@n28)) {name=Opacity, nodeColor=<rgba 0.1615 0.6287 1 1>}
  n50 = TextureSample(UVs=@n10) {samplerType=SAMPLERTYPE_Normal, texture=/Game/Characters/Player/Female/Medium/Bodies/F_MED_Bistro_Astronaut/FX/T_GaussianNoise_Normal}
  NamedRerouteDeclaration(None=@n50.RGB) {name=Normal, nodeColor=<rgba 0.0352 0.1636 1 1>}
  NamedRerouteUsage()
  NamedRerouteUsage()
  TextureCoordinate()
  TextureCoordinate()
  TextureCoordinate()
  TextureCoordinate()
 ]
}
