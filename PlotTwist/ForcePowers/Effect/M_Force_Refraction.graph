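# Additive unlit Niagara sprite material for a Force power effect, producing a refraction distortion.
# A Gaussian-noise normal map, offset per particle, supplies the distortion normal and breaks up a soft circular mask's radius.
# That mask, particle alpha and an inverted Near_Camera_Fade driven by DistanceFade scale a slight index-of-refraction shift, limiting it by camera distance.
# Shading path switches give forward and mobile a neutral index and flat normal, disabling distortion there.
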Material /PlotTwist/ForcePowers/Effect/M_Force_Refraction
graph {
 properties {
  bUsedWithNiagaraSprites=true
  blendMode=BLEND_Additive
  pixelDepthOffsetMode=PDOM_Legacy
  refractionMethod=RM_IndexOfRefraction
  shadingModel=MSM_Unlit
  usedShadingModels=Unlit
 }
 nodes [
  Constant()
  Constant() {r=1}
  n1 = TextureSample(UVs=Add(A=TextureCoordinate(), B=ParticleRandom())) {samplerType=SAMPLERTYPE_Normal, texture=/Game/Characters/Player/Female/Medium/Bodies/F_MED_Bistro_Astronaut/FX/T_GaussianNoise_Normal}
  n2 = SphereMask(A=TextureCoordinate(), B=Constant() {r=0.5}, Radius=LinearInterpolate(Alpha=@n1.R) {constA=0.5, constB=0.7}) {attenuationRadius=0.5, hardnessPercent=0}
  n3 = ParticleColor()
  n4 = Saturate(None=MaterialFunctionCall(Distance=ScalarParameter() {defaultValue=3500, parameterName=DistanceFade}) {materialFunction=&Near_Camera_Fade}.Result)
  n5 = LinearInterpolate(Alpha=Multiply(A=Multiply(A=@n2, B=@n3.A), B=LinearInterpolate(Alpha=@n4) {constA=1, constB=0})) {constA=1, constB=1.1}
  n6 = Constant() {r=1}
  ShadingPathSwitch(Default=@n5, Deferred=@n5, Forward=@n6, Mobile=@n6)
  n7 = Constant3Vector() {constant=<rgba 0 0 1 0>}
  ShadingPathSwitch(Default=@n1.RGB, Deferred=@n1.RGB, Forward=@n7, Mobile=@n7)
 ]
}
